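# Sound cue for the forward thrust of the BengalBasher glider, layering a one-shot burst over a sustained loop.
# A Mixer blends a random pick of three thrust one-shots, given slight pitch and volume modulation, with a quieter random pick of three looping waves.
# An Enveloper ramps the loop's volume up from silence so it swells in beneath the burst.
# Glider-specific attenuation, concurrency and audio property sheet settings govern playback.
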SoundCue /BRCosmetics/Sounds/Gliders/BengalBasher/GliderThrustFwd_BengalBasher_Cue
root=Mixer_0
nodes [
 {
  id=Enveloper_0
  children [Random_1]
  properties {pitchCurve {editorCurveData {keys [{time=0}]}}, volumeCurve {editorCurveData {keys [<at 0 0>,<at 1 1>]}}}
 }
 {id=Mixer_0, children [Modulator_0,Enveloper_0], properties {inputVolume [1,0.6]}}
 {id=Modulator_0, children [Random_0], properties {pitchMax=1.02, pitchMin=1, volumeMax=1.02, volumeMin=0.98}}
 {id=Random_0, children [WavePlayer_0,WavePlayer_1,WavePlayer_2], properties {weights [1,1,1]}}
 {id=Random_1, children [WavePlayer_3,WavePlayer_4,WavePlayer_5], properties {weights [1,1,1]}}
 {id=WavePlayer_0, properties {soundWaveAssetPtr=/Game/Sounds/Fort_Gadgets/Athena/BengalBasher/Glider_BengalBasher_ThrustFwd_01}}
 {id=WavePlayer_1, properties {soundWaveAssetPtr=/Game/Sounds/Fort_Gadgets/Athena/BengalBasher/Glider_BengalBasher_ThrustFwd_03}}
 {id=WavePlayer_2, properties {soundWaveAssetPtr=/Game/Sounds/Fort_Gadgets/Athena/BengalBasher/Glider_BengalBasher_ThrustFwd_02}}
 {id=WavePlayer_3, properties {bLooping=true, soundWaveAssetPtr=&/Glider_BengalBasher_Thrust_LP_Fwd}}
 {id=WavePlayer_4, properties {bLooping=true, soundWaveAssetPtr=&/Glider_BengalBasher_Thrust_LP_Fwd_02}}
 {id=WavePlayer_5, properties {bLooping=true, soundWaveAssetPtr=&/Glider_BengalBasher_Thrust_LP_Fwd_03}}
]
properties {
 attenuationSettings=&Fort_Player_Glider_Thrust_Athena_Att
 audioPropertiesSheet=&APS_Cosmetic_Glider_Loop
 maxDistance=12500
 soundClassObject=&Fort_Gliders_Land
 volumeMultiplier=0.8
 concurrencySet [&Athena_Glider_Thrust_Conc]
}
