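# Looping thrust sound cue for the BengalBasher glider, with its layer balance driven by movement input.
# A mixer blends idle, idle-layer and rattle waves, each through a continuous modulator reading the AnyDirection parameter.
# As that parameter rises, the idle layer fades out, the main idle ducks, the rattle swells and all three rise slightly in pitch.
# It uses the player glider thrust attenuation and cosmetic glider loop properties.
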SoundCue /BRCosmetics/Sounds/Gliders/BengalBasher/GliderThrustLoop_BengalBasher_Cue
root=Mixer_0
nodes [
 {id=Mixer_0, children [ModulatorContinuous_1, ModulatorContinuous_4, ModulatorContinuous_0], properties {inputVolume [1,0.45,1]}}
 {
  id=ModulatorContinuous_0
  children [WavePlayer_0]
  properties {
   pitchModulationParams {default=0, minOutput=0.9, parameterName=AnyDirection}
   volumeModulationParams {default=0, maxOutput=0, minOutput=1, parameterName=AnyDirection}
  }
 }
 {
  id=ModulatorContinuous_1
  children [WavePlayer_3]
  properties {
   pitchModulationParams {default=0, minOutput=0.9, parameterName=AnyDirection}
   volumeModulationParams {default=0, maxOutput=0.4, minOutput=0.8, parameterName=AnyDirection}
  }
 }
 {
  id=ModulatorContinuous_4
  children [WavePlayer_4]
  properties {
   pitchModulationParams {default=0, minOutput=0.9, parameterName=AnyDirection}
   volumeModulationParams {default=0, minOutput=0.3, parameterName=AnyDirection}
  }
 }
 {
  id=WavePlayer_0
  properties {bLooping=true, soundWaveAssetPtr=&/Glider_BengalBasher_Thrust_LP_Idle_Lyr}
 }
 {id=WavePlayer_3, properties {bLooping=true, soundWaveAssetPtr=&/Glider_BengalBasher_Thrust_LP_Idle}}
 {id=WavePlayer_4, properties {bLooping=true, soundWaveAssetPtr=&/Glider_BengalBasher_Thrust_LP_Rattle}}
]
properties {
 attenuationSettings=&Fort_Player_Glider_Thrust_Athena_Att
 audioPropertiesSheet=&APS_Cosmetic_Glider_Loop
 maxDistance=12500
 soundClassObject=&Fort_Gliders_Land
 volumeMultiplier=0.7
 concurrencySet [null]
}
